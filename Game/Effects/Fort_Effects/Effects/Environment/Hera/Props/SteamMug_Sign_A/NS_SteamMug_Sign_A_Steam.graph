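# Niagara system that emits steam for the SteamMug_Sign_A environment prop.
# SteamUpward, the only enabled emitter, continuously spawns sprites in a rotated cylinder, nudges them along X and lifts them with upward acceleration against drag.
# Sprites grow, fade and shift from warm tint to white, using SubUV animation, viewpoint-based rotation and a curve-driven dynamic material parameter.
# A disabled SteamOutward emitter holds a radial ring-burst variant; both derive their scale from System Global Scale.
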NiagaraSystem /Game/Effects/Fort_Effects/Effects/Environment/Hera/Props/SteamMug_Sign_A/NS_SteamMug_Sign_A_Steam
dependencies {
 usedDataInterfaces [
  &NiagaraDataInterfaceColorCurve
  &NiagaraDataInterfaceCurve
  &NiagaraDataInterfaceCamera
  &NiagaraDataInterfaceSpriteRendererInfo
 ]
 usedDynamicInputs [&Vector2DFromFloat]
 usedModules [
  &self:SystemSpawnScript.SystemScriptSource.SystemScriptGraph.NiagaraNodeAssignment_0.SetVariables_#1
  &SystemState
 ]
}
emitters {
 SteamOutward {
  bEnabled=false
  emitterSpawnScript [/SetVariables(Emitter.Global Scale:NiagaraFloat=&Multiply_Float(A=$System.Global Scale, B=0.8)) {isSetParameters=true, module=SetVariables_#2}]
  emitterUpdateScript [
   &EmitterState(Life Cycle Mode=::NewEnumerator1, Loop Behavior=::NewEnumerator1, Loop Delay=0.02, Loop Duration=1.8761, UseLoopDelay=true)
   &SpawnBurst_Instantaneous(Spawn Count=17, Spawn Time=0.05, Use Spawn Probability=false)
   /SetVariables(
    Emitter. Color:LinearColor=&ColorFromCurve(
     ColorCurve=DataInterface() {propertyValues {
       AlphaCurve {keys [<at 0 1 RCIM_Cubic>,<at 0.1176 0 RCIM_Cubic>]}
       BlueCurve {keys [
         <at 0 1.9167>
         {arriveTangent=-15.4449, interpMode=RCIM_Cubic, leaveTangent=-15.4449, time=0.0418, value=0}
         <at 0.0836 0.625>
         <at 0.1176 0>
         <at 1 0 RCIM_Cubic>
        ]}
       GreenCurve {keys [
         <at 0 6.3255>
         {arriveTangent=4.3634, interpMode=RCIM_Cubic, leaveTangent=4.3634, time=0.0418, value=7.7363}
         <at 0.0836 6.6904>
         <at 0.1176 0>
         <at 1 0 RCIM_Cubic>
        ]}
       RedCurve {keys [
         <at 0 16>
         {arriveTangent=47.8293, interpMode=RCIM_Cubic, leaveTangent=47.8293, time=0.0418, value=50}
         <at 0.0836 20>
         <at 0.1176 0>
         <at 1 0 RCIM_Cubic>
        ]}
      }}
     CurveIndex=$Emitter.NormalizedLoopAge
    )
    Emitter.Exponent=1.5
    Emitter.Position:NiagaraPosition=&Multiply_VectorByFloat(
     Vector=&AddVectorToPosition(
      Position=$Engine.Owner.Position
      Vector=&MakeVector(
       X=0
       Y=0
       Z=&Multiply_Float(
        A=$Emitter.Global Scale
        B=&FloatFromCurve(
         FloatCurve=DataInterface() {propertyValues {Curve {keys [{arriveTangent=0.3523, interpMode=RCIM_Cubic, leaveTangent=0.3523, tangentMode=RCTM_User, time=0, value=1.25}, <at 0.5 1.5>]}}}
         CurveIndex=$Engine.System.Age
         Scale Curve=300
        )
       )
      )
     )
     Float=1
    )
    Emitter.Radius:NiagaraFloat=&Multiply_Float(
     A=&FloatFromCurve(
      FloatCurve=DataInterface() {propertyValues {Curve {keys [{interpMode=RCIM_Cubic, tangentMode=RCTM_User, time=0.1, value=0.95}, <at 0.5 0>]}}}
      CurveIndex=$Emitter.NormalizedLoopAge
      Scale Curve=600
     )
     B=$Emitter.Global Scale
    )
    Emitter.VisibilityTag=1
   ) {isSetParameters=true}
  ]
  particleSpawnScript [
   &InitializeParticle(
    Color=<rgba 0.1 0.1 0.1 1>
    Lifetime Max=1.4
    Lifetime Min=0.4
    Lifetime Mode=::NewEnumerator1
    Sprite Rotation Mode=::NewEnumerator2
    Sprite Size Mode=::NewEnumerator3
    Sprite UV Mode=::NewEnumerator3
    Uniform Sprite Size=&Multiply_Float(A=&RandomRangeFloat(Minimum=400, Maximum=450), B=$Emitter.Global Scale)
    UsePositionOffset=true
    Write Material Random=true
   )
   &ShapeLocation(
    MaskSpawnGroup=false
    Override Seed=false
    Radius Position=1.2
    Ring / Disc Distribution Mode:ENiagaraSphereDistributionMode=::NewEnumerator1
    Ring Radius=&Multiply_Float(A=50, B=$Emitter.Global Scale)
    Rotation Mode=::NewEnumerator1
    Scale Mode:ENiagara_ScaleMode=::NewEnumerator1
    Shape Primitive=::NewEnumerator4
    U Position=&ReturnNormalizedExecIndex(Normalized Index Includes One=false)
   )
   &AddVelocityFromPoint(
    Module.Use Falloff Distance=false
    Origin Offset=&Multiply_VectorByFloat(Vector=<0 0 -5>, Float=$Emitter.Global Scale)
    Velocity Origin:Vector3f=&SimulationPosition()
    Velocity Strength=&Multiply_Float(A=&UniformRangedFloat(Minimum=2000, Maximum=2500), B=$Emitter.Global Scale)
   )
   &ApplyInitialForces()
   &CameraOffset(Camera Offset Amount=&Multiply_Float(A=40, B=$Emitter.Global Scale))
   /SetVariables(
    Particles.Stored Particle Size:Vector3f=$Particles.Scale
    Particles.Stored Sprite Size:Vector2f=&Multiply_Vector2D_ByFloat(Float=$Emitter.Global Scale, Vector2D=$Particles.SpriteSize)
   ) {isSetParameters=true}
  ]
  particleUpdateScript [
   &ParticleState()
   /SetVariables(Particles.Scale=$StackContext.Stored Particle Size, Particles.SpriteSize=$StackContext.Stored Sprite Size) {isSetParameters=true}
   &AccelerationForce(Acceleration=&MakeVector(X=0, Y=0, Z=&UniformRangedFloat(Minimum=-100, Maximum=200)))
   &Drag(
    Drag=&Divide_Float(A=&RandomRangeFloat(Minimum=7, Maximum=9), B=$Emitter.Global Scale)
    Use Linear Drag=true
    Use Rotational Drag=true
   )
   &self:ViewpointBased_Sprite_Rotation_03(Rotation Rate=3) {module=ViewpointBased_Sprite_Rotation_03}
   &SolveForcesAndVelocity(Clamp Velocity=false, Limit Acceleration=false)
   &ScaleSpriteSize(Uniform Curve Sprite Scale=DataInterface() {propertyValues {Curve {keys [
        {arriveTangent=0.8609, interpMode=RCIM_Cubic, leaveTangent=0.8609, tangentMode=RCTM_User, time=0, value=0.15}
        {arriveTangent=2.1878, interpMode=RCIM_Cubic, leaveTangent=2.1878, tangentMode=RCTM_User, time=0.4, value=0.85}
        <at 1 1.25 RCIM_Cubic>
       ]}}})
   &DynamicMaterialParameters(
    Index 0 Param 1=&FloatFromCurve(FloatCurve=DataInterface() {propertyValues {Curve {keys [<at -0.0046 1.0113>,<at 0.2 1>,<at 1.0011 -1>]}}})
    Index 0 Param 4=0.2
    Module.Index0Float0Write=true
    Module.Index0Float1Write=true
    Module.Index0Float2Write=true
    Module.Index0Float3Write=true
   ) {module=DynamicMatParams}
   &ScaleColor(
    Scale Alpha=&FloatFromCurve(
     FloatCurve=DataInterface() {propertyValues {Curve {keys [
         {arriveTangent=6.7279, interpMode=RCIM_Cubic, leaveTangent=6.7279, tangentMode=RCTM_User, time=0.0025, value=0.5379}
         {arriveTangent=-0.5367, interpMode=RCIM_Cubic, leaveTangent=-0.5367, time=0.1843, value=0.9779}
         <at 1.0011 0.002 RCIM_Cubic>
        ]}}}
     Scale Curve=0.6
    )
    Scale RGB=&MakeVectorFromLinearColorRGB(Color=&ColorFromCurve(ColorCurve=DataInterface() {propertyValues {
        AlphaCurve {keys [<at 0 1 RCIM_Cubic>,<at 1 0 RCIM_Cubic>]}
        BlueCurve {keys [<at 0 0.5104 RCIM_Cubic>,<at 1 1 RCIM_Cubic>]}
        GreenCurve {keys [<at 0 0.6952 RCIM_Cubic>,<at 1 1 RCIM_Cubic>]}
        RedCurve {keys [<at 0 1 RCIM_Cubic>,<at 1 1 RCIM_Cubic>]}
       }}))
    ScaleA=true
    ScaleRGB=true
   )
   &SubUVAnimation(SubUV Loop Count=0.5)
  ]
  rendererClasses [&NiagaraSpriteRendererProperties]
  renderers [{rendererClass=&NiagaraSpriteRendererProperties}]
 }
 SteamUpward {
  emitterSpawnScript [/SetVariables(Emitter.Global Scale:NiagaraFloat=&Multiply_Float(A=$System.Global Scale, B=0.8)) {isSetParameters=true, module=SetVariables_#2}]
  emitterUpdateScript [
   &EmitterState(Life Cycle Mode=::NewEnumerator1, Loop Behavior=::NewEnumerator1, Loop Delay=0.02, Loop Duration=1.5, UseLoopDelay=false)
   &SpawnRate(SpawnRate=5, Use Spawn Probability=false)
   &SpawnBurst_Instantaneous(Spawn Count=9, Spawn Time=0.05, Use Spawn Probability=false, UseLoopCountLimit=false) {enabled=false}
  ]
  particleSpawnScript [
   &InitializeParticle(
    Color=<rgba 0.1 0.1 0.1 1>
    Lifetime Min=2
    Lifetime Mode=::NewEnumerator1
    Position Offset=<0 0 10>
    Sprite Rotation Mode=::NewEnumerator2
    Sprite Size Mode=::NewEnumerator3
    Sprite UV Mode=::NewEnumerator3
    Uniform Sprite Size=&Multiply_Float(A=&RandomRangeFloat(Minimum=450, Maximum=450), B=$Emitter.Global Scale)
    UsePositionOffset=true
    Write Material Random=true
   )
   &ShapeLocation(
    Cylinder Height=2
    Cylinder Height Midpoint=0
    Cylinder Radius=60
    MaskSpawnGroup=false
    Override Seed=false
    Rotation Angle=90
    Rotation Angle Type:ENiagara_AngleInput=::NewEnumerator0
    Rotation Mode=::NewEnumerator1
    Scale Mode:ENiagara_ScaleMode=::NewEnumerator1
    Shape Primitive=::NewEnumerator1
    Surface Only=false
   )
   &CameraOffset(Camera Offset Amount=&Multiply_Float(A=-10, B=$Emitter.Global Scale))
   &AddVelocity(Velocity=<1 0 0>, Velocity Speed Scale=100) {module=AddVelocity001}
   &ApplyInitialForces(Apply Rotational Force to Rotational Velocity=false)
  ]
  particleUpdateScript [
   &ParticleState()
   &AccelerationForce(Acceleration=&MakeVector(
     X=0
     Y=0
     Z=&Multiply_Float(
      A=&Multiply_Float(A=100, B=&RandomRangeFloat(Minimum=0.8, Maximum=1))
      B=$System.Global Scale
     )
    ))
   &Drag(Drag=&Divide_Float(A=0.5, B=$Emitter.Global Scale), Use Linear Drag=true, Use Rotational Drag=false)
   &self:ViewpointBased_Sprite_Rotation_04(Rotation Rate=2) {module=ViewpointBased_Sprite_Rotation_04}
   &SolveForcesAndVelocity(Clamp Velocity=true, Limit Acceleration=false, Speed Limit=500)
   &ScaleSpriteSize(
    Uniform Curve Scale=2
    Uniform Curve Sprite Scale=DataInterface() {propertyValues {Curve {keys [
        {arriveTangent=0.0751, interpMode=RCIM_Cubic, leaveTangent=0.0751, tangentMode=RCTM_User, time=0, value=0.4}
        {arriveTangent=1.2128, interpMode=RCIM_Cubic, leaveTangent=1.2128, tangentMode=RCTM_User, time=0.3871, value=0.6295}
        {arriveTangent=0.0206, interpMode=RCIM_Cubic, leaveTangent=0.0206, tangentMode=RCTM_User, time=1, value=1}
       ]}}}
   )
   &DynamicMaterialParameters(
    Index 0 Param 1=&FloatFromCurve(FloatCurve=DataInterface() {propertyValues {Curve {keys [<at -0.0046 1.0113>,<at 0.2 1>,<at 1.0011 -1>]}}})
    Index 0 Param 4=0.2
    Module.Param0WriteEnabled=true
    Module.Param1WriteEnabled=false
    Module.Param2WriteEnabled=false
    Module.Param3WriteEnabled=true
   ) {module=DynamicMatParams}
   &ScaleColor(
    Scale Alpha=&FloatFromCurve(FloatCurve=DataInterface() {propertyValues {Curve {keys [
         {arriveTangent=0.0655, interpMode=RCIM_Cubic, leaveTangent=0.0655, tangentMode=RCTM_User, time=0, value=0}
         {arriveTangent=0.4667, interpMode=RCIM_Cubic, leaveTangent=0.4667, time=0.3474, value=1}
         {arriveTangent=-1.5322, interpMode=RCIM_Cubic, leaveTangent=-1.5322, time=0.6053, value=0.2825}
         <at 1 0 RCIM_Cubic>
        ]}}})
    Scale RGB=&MakeVectorFromLinearColorRGB(Color=&ColorFromCurve(ColorCurve=DataInterface() {propertyValues {
        AlphaCurve {keys [<at 0 1 RCIM_Cubic>,<at 1 0 RCIM_Cubic>]}
        BlueCurve {keys [<at 0 0.5104 RCIM_Cubic>,<at 1 1 RCIM_Cubic>]}
        GreenCurve {keys [<at 0 0.6952 RCIM_Cubic>,<at 1 1 RCIM_Cubic>]}
        RedCurve {keys [<at 0 1 RCIM_Cubic>,<at 1 1 RCIM_Cubic>]}
       }}))
    ScaleA=true
    ScaleRGB=true
   )
   &SubUVAnimation(
    "Sprite Renderer "=DataInterface() {propertyValues {SpriteRenderer=&self:SteamUpward_0.NiagaraSpriteRendererProperties_1}}
    SubUV Lookup Index Scale=0.5
    UseEndFrame=true
    UseStartFrame=true
   )
  ]
  rendererClasses [&NiagaraSpriteRendererProperties]
  renderers [{rendererClass=&NiagaraSpriteRendererProperties}]
 }
}
summary {
 systemName=NS_SteamMug_Sign_A_Steam
 emitters [
  {bEnabled=false, emitterName=SteamOutward, rendererClasses [&NiagaraSpriteRendererProperties]}
  {emitterName=SteamUpward, rendererClasses [&NiagaraSpriteRendererProperties]}
 ]
 userVariables []
}
